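# Control Rig for the M_MED_TroopFactor cosmetic body, posing arm and leg deform spline bones plus knee and groin correctives from the animated skeleton.
# Weighted constraints place control nulls that drive stretchy B-spline IK chains, with lists filled at construction.
# Smoothed arm-hidden flags, combined with spherical pose readers, offset arm controls.
# Remapped thigh/calf Z rotations and smoothed crouch state weight leg offsets; a disabled debug branch copies controls onto bones.
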ControlRigBlueprint /BRCosmetics/Characters/Player/Male/Medium/Bodies/M_MED_TroopFactor/Meshes/M_MED_TroopFactor_CtrlRig
graphs [
 New Function_ContainedGraph {nodes [
   RigVMFunctionEntryNode()
   > RigVMFunction_Sequence = RigVMAggregateNode()
   &RigUnit_ParentConstraint(
    AdvancedSettings=(InterpolationType=Shortest,RotationOrderForFilter=XZY)
    Child=(Type=Null,Name="dfrm_arm_spline_01_l_ctrl_null")
    Filter=(TranslationFilter=(bX=false,bY=false,bZ=false),RotationFilter=(bX=True,bY=True,bZ=True),ScaleFilter=(bX=True,bY=True,bZ=True))
    Parents=((Item=(Type=Bone,Name="upperarm_l"),Weight=0.3),(Item=(Type=Bone,Name="spine_05"),Weight=0.7))
    Weight=1
    ExecutePin=@RigVMFunction_Sequence.A
   )
   > &RigUnit_ParentConstraint(
    AdvancedSettings=(InterpolationType=Shortest,RotationOrderForFilter=XZY)
    Child=(Type=Null,Name="dfrm_arm_spline_05_l_ctrl_null")
    Parents=((Item=(Type=Control,Name="dfrm_arm_spline_01_l_ctrl"),Weight=0.3),(Item=(Type=Bone,Name="upperarm_l"),Weight=0.35),(Item=(Type=Bone,Name="lowerarm_l"),Weight=0.35))
    Weight=1
   )
   > &RigUnit_ParentConstraint(
    AdvancedSettings=(InterpolationType=Shortest,RotationOrderForFilter=XZY)
    Child=(Type=Null,Name="dfrm_arm_spline_03_l_ctrl_null")
    Parents=((Item=(Type=Control,Name="dfrm_arm_spline_01_l_ctrl"),Weight=0.7),(Item=(Type=Control,Name="dfrm_arm_spline_05_l_ctrl"),Weight=0.3))
    Weight=1
   )
   > &RigUnit_ParentConstraint(
    AdvancedSettings=(InterpolationType=Shortest,RotationOrderForFilter=XZY)
    Child=(Type=Null,Name="dfrm_arm_spline_09_l_ctrl_null")
    Filter=(TranslationFilter=(bX=True,bY=True,bZ=True),RotationFilter=(bX=True,bY=True,bZ=True),ScaleFilter=(bX=true,bY=true,bZ=true))
    Parents=((Item=(Type=Bone,Name="hand_l"),Weight=0.7),(Item=(Type=Bone,Name="lowerarm_l"),Weight=0.3))
    Weight=1
   )
   > &RigUnit_ParentConstraint(
    AdvancedSettings=(InterpolationType=Shortest,RotationOrderForFilter=XZY)
    Child=(Type=Null,Name="dfrm_arm_spline_07_l_ctrl_null")
    Parents=((Item=(Type=Control,Name="dfrm_arm_spline_05_l_ctrl"),Weight=0.5),(Item=(Type=Control,Name="dfrm_arm_spline_09_l_ctrl"),Weight=0.5))
    Weight=1
   )
   RigVMFunction_MathBoolToFloat = &RigVMFunction_MathBoolToFloat(Value=RigVMVariableNode() {variable=L_armHidden}.Value)
   n1 = &RigVMFunction_AlphaInterp(
    InterpSpeedDecreasing=10
    InterpSpeedIncreasing=10
    OutRange=(Min=0,Max=1.4)
    Scale=1
    bClampResult=true
    Value=@RigVMFunction_MathBoolToFloat.Result
   )
   > &RigUnit_ModifyTransforms(
    ItemToModify="((Item=(Type=Bone,Name=\"None\"),Transform=(Rotation=(X=0,Y=0,Z=0,W=1),Translation=(X=0,Y=0,Z=0),Scale3D=(X=1,Y=1,Z=1))),(Item=(Type=Null,Name=\"dfrm_arm_spline_05_l_ctrl_null\"),Transform=(Rotation=(X=0,Y=0,Z=0,W=-1),Translation=(X=-2.8743,Y=-3.2013,Z=-2.263) ...(truncated)"
    WeightMinimum=0
    Weight=@n1.Result
   )
   > n2 = &RigUnit_SphericalPoseReader(
    ActiveRegionScaleFactors=(PositiveWidth=1,NegativeWidth=1,PositiveHeight=1,NegativeHeight=1)
    ActiveRegionSize=0.25
    Debug=(bDrawDebug=false,bDraw2D=False,bDrawLocalAxes=False,DebugScale=25,DebugSegments=20,DebugThickness=0.25)
    DriverAxis=(X=1,Y=0,Z=0)
    DriverItem=(Type=Bone,Name="upperarm_l")
    FalloffRegionScaleFactors=(PositiveWidth=1,NegativeWidth=1,PositiveHeight=1,NegativeHeight=1)
    FalloffSize=0.45
    FlipHeightScaling=False
    FlipWidthScaling=False
    OptionalParentItem=(Type=Bone,Name="clavicle_l")
    RotationOffset=(X=5.0606,Y=-28.9001,Z=-11.9397)
   )
   Multiply = &RigVMFunction_MathDoubleMul(B=@n1.Result, A=@n2.OutputParam)
   > &RigUnit_ModifyTransforms(
    ItemToModify="((Item=(Type=Bone,Name=\"None\"),Transform=(Rotation=(X=0,Y=0,Z=0,W=1),Translation=(X=0,Y=0,Z=0),Scale3D=(X=1,Y=1,Z=1))),(Item=(Type=Null,Name=\"dfrm_arm_spline_07_l_ctrl_null\"),Transform=(Rotation=(X=0,Y=0,Z=0,W=-1),Translation=(X=-5.908,Y=-0.2369,Z=4.9091), ...(truncated)"
    WeightMinimum=0
    Weight=@Multiply.Result
   )
   ItemArray = RigVMDispatchNode(Value="((Type=Bone,Name=\"dfrm_arm_spline_01_l\"),(Type=Bone,Name=\"dfrm_arm_spline_02_l\"),(Type=Bone,Name=\"dfrm_arm_spline_03_l\"),(Type=Bone,Name=\"dfrm_arm_spline_04_l\"),(Type=Bone,Name=\"dfrm_arm_spline_05_l\"),(Type=Bone,Name=\"dfrm_arm_spline_06_l\"),(Type=Bone,Name ...(truncated)") {template="DISPATCH_RigVMDispatch_Constant(io Value)"}
   > SplineIK = RigVMFunctionReferenceNode(
    Primary Axis=(X=1,Y=0,Z=0)
    Samples Per Segment=16
    Secondary Spline Direction=(X=0,Y=0,Z=1)
    Spline Mode=BSpline
    Stretch=true
    Up Axis=(X=0,Y=0,Z=1)
    Bones=@ItemArray.Value
    Controls=RigVMVariableNode() {variable=L_armSpline_ctrls}.Value
   )
   &RigUnit_ParentConstraint(
    AdvancedSettings=(InterpolationType=Shortest,RotationOrderForFilter=XZY)
    Child=(Type=Null,Name="dfrm_arm_spline_01_r_ctrl_null")
    Filter=(TranslationFilter=(bX=false,bY=false,bZ=false),RotationFilter=(bX=True,bY=True,bZ=True),ScaleFilter=(bX=True,bY=True,bZ=True))
    Parents=((Item=(Type=Bone,Name="upperarm_r"),Weight=0.3),(Item=(Type=Bone,Name="spine_05"),Weight=0.7))
    Weight=1
    ExecutePin=@RigVMFunction_Sequence.B
   )
   > &RigUnit_ParentConstraint(
    AdvancedSettings=(InterpolationType=Shortest,RotationOrderForFilter=XZY)
    Child=(Type=Null,Name="dfrm_arm_spline_05_r_ctrl_null")
    Parents=((Item=(Type=Control,Name="dfrm_arm_spline_01_r_ctrl"),Weight=0.3),(Item=(Type=Bone,Name="upperarm_r"),Weight=0.35),(Item=(Type=Bone,Name="lowerarm_r"),Weight=0.35))
    Weight=1
   )
   > &RigUnit_ParentConstraint(
    AdvancedSettings=(InterpolationType=Shortest,RotationOrderForFilter=XZY)
    Child=(Type=Null,Name="dfrm_arm_spline_03_r_ctrl_null")
    Parents=((Item=(Type=Control,Name="dfrm_arm_spline_01_r_ctrl"),Weight=0.7),(Item=(Type=Control,Name="dfrm_arm_spline_05_r_ctrl"),Weight=0.3))
    Weight=1
   )
   > &RigUnit_ParentConstraint(
    AdvancedSettings=(InterpolationType=Shortest,RotationOrderForFilter=XZY)
    Child=(Type=Null,Name="dfrm_arm_spline_09_r_ctrl_null")
    Filter=(TranslationFilter=(bX=True,bY=True,bZ=True),RotationFilter=(bX=True,bY=True,bZ=True),ScaleFilter=(bX=true,bY=true,bZ=true))
    Parents=((Item=(Type=Bone,Name="hand_r"),Weight=0.7),(Item=(Type=Bone,Name="lowerarm_r"),Weight=0.3))
    Weight=1
   )
   > &RigUnit_ParentConstraint(
    AdvancedSettings=(InterpolationType=Shortest,RotationOrderForFilter=XZY)
    Child=(Type=Null,Name="dfrm_arm_spline_07_r_ctrl_null")
    Parents=((Item=(Type=Control,Name="dfrm_arm_spline_05_r_ctrl"),Weight=0.5),(Item=(Type=Control,Name="dfrm_arm_spline_09_r_ctrl"),Weight=0.5))
    Weight=1
   )
   n3 = &RigVMFunction_AlphaInterp(
    InterpSpeedDecreasing=10
    InterpSpeedIncreasing=10
    OutRange=(Min=0,Max=1.4)
    Scale=1
    bClampResult=true
    Value=&RigVMFunction_MathBoolToFloat(Value=RigVMVariableNode() {variable=R_armHidden}.Value).Result
   )
   > &RigUnit_ModifyTransforms(
    ItemToModify="((Item=(Type=Bone,Name=\"None\"),Transform=(Rotation=(X=0,Y=0,Z=0,W=1),Translation=(X=0,Y=0,Z=0),Scale3D=(X=1,Y=1,Z=1))),(Item=(Type=Null,Name=\"dfrm_arm_spline_05_r_ctrl_null\"),Transform=(Rotation=(X=0,Y=0,Z=0,W=-1),Translation=(X=5.6513,Y=8.489,Z=2.7159),Sc ...(truncated)"
    WeightMinimum=0
    Weight=@n3.Result
   )
   > n4 = &RigUnit_SphericalPoseReader(
    ActiveRegionScaleFactors=(PositiveWidth=1,NegativeWidth=1,PositiveHeight=1,NegativeHeight=1)
    ActiveRegionSize=0.25
    Debug=(bDrawDebug=false,bDraw2D=False,bDrawLocalAxes=False,DebugScale=25,DebugSegments=20,DebugThickness=0.25)
    DriverAxis=(X=-1,Y=0,Z=0)
    DriverItem=(Type=Bone,Name="upperarm_r")
    FalloffRegionScaleFactors=(PositiveWidth=1,NegativeWidth=1,PositiveHeight=1,NegativeHeight=1)
    FalloffSize=0.45
    FlipHeightScaling=False
    FlipWidthScaling=False
    OptionalParentItem=(Type=Bone,Name="clavicle_r")
    RotationOffset=(X=-0.8456,Y=160.0927,Z=-1.437)
   )
   > &RigUnit_ModifyTransforms(
    ItemToModify="((Item=(Type=Bone,Name=\"None\"),Transform=(Rotation=(X=0,Y=0,Z=0,W=1),Translation=(X=0,Y=0,Z=0),Scale3D=(X=1,Y=1,Z=1))),(Item=(Type=Null,Name=\"dfrm_arm_spline_07_r_ctrl_null\"),Transform=(Rotation=(X=0,Y=0,Z=0,W=-1),Translation=(X=5.7523,Y=0.6013,Z=-3.1804), ...(truncated)"
    WeightMinimum=0
    Weight=&RigVMFunction_MathDoubleMul(B=@n3.Result, A=@n4.OutputParam).Result
   )
   ItemArray_1 = RigVMDispatchNode(Value="((Type=Bone,Name=\"dfrm_arm_spline_01_r\"),(Type=Bone,Name=\"dfrm_arm_spline_02_r\"),(Type=Bone,Name=\"dfrm_arm_spline_03_r\"),(Type=Bone,Name=\"dfrm_arm_spline_04_r\"),(Type=Bone,Name=\"dfrm_arm_spline_05_r\"),(Type=Bone,Name=\"dfrm_arm_spline_06_r\"),(Type=Bone,Name ...(truncated)") {template="DISPATCH_RigVMDispatch_Constant(io Value)"}
   > SplineIK_1 = RigVMFunctionReferenceNode(
    Primary Axis=(X=-1,Y=0,Z=0)
    Samples Per Segment=16
    Secondary Spline Direction=(X=0,Y=0,Z=1)
    Spline Mode=BSpline
    Stretch=true
    Up Axis=(X=0,Y=0,Z=1)
    Bones=@ItemArray_1.Value
    Controls=RigVMVariableNode() {variable=R_armSpline_ctrls}.Value
   )
   RigVMFunctionReturnNode(ExecuteContext=@RigVMFunction_Sequence.C)
  ]}
 New Function_ContainedGraph {nodes [
   RigVMFunctionEntryNode()
   > RigVMFunction_Sequence = RigVMAggregateNode()
   &RigUnit_ParentConstraint(
    AdvancedSettings=(InterpolationType=Shortest,RotationOrderForFilter=XZY)
    Child=(Type=Bone,Name="dfrm_thigh_l")
    Filter=(TranslationFilter=(bX=true,bY=true,bZ=true),RotationFilter=(bX=True,bY=True,bZ=True),ScaleFilter=(bX=True,bY=True,bZ=True))
    Parents=((Item=(Type=Bone,Name="thigh_l"),Weight=1))
    Weight=1
    ExecutePin=@RigVMFunction_Sequence.A
   )
   > &RigUnit_ParentConstraint(
    AdvancedSettings=(InterpolationType=Shortest,RotationOrderForFilter=XZY)
    Child=(Type=Bone,Name="dfrm_calf_l")
    Filter=(TranslationFilter=(bX=true,bY=true,bZ=true),RotationFilter=(bX=True,bY=True,bZ=True),ScaleFilter=(bX=True,bY=True,bZ=True))
    Parents=((Item=(Type=Bone,Name="calf_l"),Weight=1))
    Weight=1
   )
   > &RigUnit_ParentConstraint(
    AdvancedSettings=(InterpolationType=Shortest,RotationOrderForFilter=XZY)
    Child=(Type=Bone,Name="dfrm_foot_l")
    Filter=(TranslationFilter=(bX=true,bY=true,bZ=true),RotationFilter=(bX=True,bY=True,bZ=True),ScaleFilter=(bX=True,bY=True,bZ=True))
    Parents=((Item=(Type=Bone,Name="foot_l"),Weight=1))
    Weight=1
   )
   > &RigUnit_ParentConstraint(
    AdvancedSettings=(InterpolationType=Shortest,RotationOrderForFilter=XZY)
    Child=(Type=Bone,Name="dfrm_thigh_r")
    Filter=(TranslationFilter=(bX=true,bY=true,bZ=true),RotationFilter=(bX=True,bY=True,bZ=True),ScaleFilter=(bX=True,bY=True,bZ=True))
    Parents=((Item=(Type=Bone,Name="thigh_r"),Weight=1))
    Weight=1
   )
   > &RigUnit_ParentConstraint(
    AdvancedSettings=(InterpolationType=Shortest,RotationOrderForFilter=XZY)
    Child=(Type=Bone,Name="dfrm_calf_r")
    Filter=(TranslationFilter=(bX=true,bY=true,bZ=true),RotationFilter=(bX=True,bY=True,bZ=True),ScaleFilter=(bX=True,bY=True,bZ=True))
    Parents=((Item=(Type=Bone,Name="calf_r"),Weight=1))
    Weight=1
   )
   > &RigUnit_ParentConstraint(
    AdvancedSettings=(InterpolationType=Shortest,RotationOrderForFilter=XZY)
    Child=(Type=Bone,Name="dfrm_foot_r")
    Filter=(TranslationFilter=(bX=true,bY=true,bZ=true),RotationFilter=(bX=True,bY=True,bZ=True),ScaleFilter=(bX=True,bY=True,bZ=True))
    Parents=((Item=(Type=Bone,Name="foot_r"),Weight=1))
    Weight=1
   )
   > Delta Rotation Z = RigVMFunctionReferenceNode(Child=calf_l, Parent=calf_l_null, Source Maximum=90, Source Minimum=0)
   > RigVMVariableNode(Value=@"Delta Rotation Z".Delta Rotation) {variable=L_calfUp}
   > Delta Rotation Z_1 = RigVMFunctionReferenceNode(Child=calf_r, Parent=calf_r_null, Source Maximum=90, Source Minimum=0)
   > RigVMVariableNode(Value=@"Delta Rotation Z_1".Delta Rotation) {variable=R_calfUp}
   > Delta Rotation Z_2 = RigVMFunctionReferenceNode(Child=thigh_l, Parent=thigh_l_null, Source Maximum=-60, Source Minimum=-25)
   > RigVMVariableNode(Value=@"Delta Rotation Z_2".Delta Rotation) {variable=L_thighUp}
   > Delta Rotation Z_3 = RigVMFunctionReferenceNode(Child=thigh_r, Parent=thigh_r_null, Source Maximum=-60, Source Minimum=-25)
   > RigVMVariableNode(Value=@"Delta Rotation Z_3".Delta Rotation) {variable=R_thighUp}
   &RigUnit_RotationConstraintLocalSpaceOffset(
    AdvancedSettings=(RotationOrderForFilter=XZY)
    Child=(Type=Null,Name="dfrm_leg_spline_01_l_ctrl_null")
    Filter=(bX=True,bY=True,bZ=True)
    Parents=((Item=(Type=Bone,Name="dfrm_thigh_l"),Weight=0.9),(Item=(Type=Bone,Name="pelvis"),Weight=0.1))
    Weight=1
    ExecutePin=@RigVMFunction_Sequence.B
   )
   > &RigUnit_ParentConstraint(
    AdvancedSettings=(RotationOrderForFilter=XZY)
    Child=(Type=Null,Name="dfrm_leg_spline_03_l_ctrl_null")
    Parents=((Item=(Type=Bone,Name="dfrm_thigh_l"),Weight=1),(Item=(Type=Bone,Name="dfrm_calf_l"),Weight=0))
    Weight=1
   )
   > &RigUnit_ParentConstraint(
    AdvancedSettings=(RotationOrderForFilter=XZY)
    Child=(Type=Null,Name="dfrm_leg_spline_05_l_ctrl_null")
    Parents=((Item=(Type=Bone,Name="dfrm_thigh_l"),Weight=0.5),(Item=(Type=Bone,Name="dfrm_calf_l"),Weight=0.5))
    Weight=1
   )
   > &RigUnit_ParentConstraint(
    AdvancedSettings=(RotationOrderForFilter=XZY)
    Child=(Type=Null,Name="dfrm_leg_spline_07_l_ctrl_null")
    Parents=((Item=(Type=Bone,Name="dfrm_calf_l"),Weight=1),(Item=(Type=Bone,Name="dfrm_foot_l"),Weight=0))
    Weight=1
   )
   > &RigUnit_ParentConstraint(
    AdvancedSettings=(RotationOrderForFilter=XZY)
    Child=(Type=Null,Name="dfrm_leg_spline_09_l_ctrl_null")
    Parents=((Item=(Type=Bone,Name="dfrm_calf_l"),Weight=0),(Item=(Type=Bone,Name="dfrm_foot_l"),Weight=1))
    Weight=1
   )
   n1 = RigVMVariableNode() {variable=L_calfUp}
   > &RigUnit_ModifyTransforms(
    ItemToModify="((Item=(Type=Bone,Name=\"None\"),Transform=(Rotation=(X=0,Y=0,Z=0,W=1),Translation=(X=0,Y=0,Z=0),Scale3D=(X=1,Y=1,Z=1))),(Item=(Type=Null,Name=\"dfrm_leg_spline_05_l_ctrl_null\"),Transform=(Rotation=(X=0,Y=0,Z=0,W=-1),Translation=(X=14.1162,Y=9.5579,Z=0),Scale ...(truncated)"
    WeightMinimum=0
    Weight=@n1.Value
   )
   ItemArray = RigVMDispatchNode(Value="((Type=Bone,Name=\"dfrm_leg_spline_01_l\"),(Type=Bone,Name=\"dfrm_leg_spline_02_l\"),(Type=Bone,Name=\"dfrm_leg_spline_03_l\"),(Type=Bone,Name=\"dfrm_leg_spline_04_l\"),(Type=Bone,Name=\"dfrm_leg_spline_05_l\"),(Type=Bone,Name=\"dfrm_leg_spline_06_l\"),(Type=Bone,Name ...(truncated)") {template="DISPATCH_RigVMDispatch_Constant(io Value)"}
   > SplineIK = RigVMFunctionReferenceNode(
    Primary Axis=(X=1,Y=0,Z=0)
    Samples Per Segment=16
    Secondary Spline Direction=(X=0,Y=0,Z=0)
    Spline Mode=BSpline
    Stretch=true
    Up Axis=(X=0,Y=0,Z=0)
    debug=false
    Bones=@ItemArray.Value
    Controls=RigVMVariableNode() {variable=L_legSpline_ctrls}.Value
   )
   > &RigUnit_ModifyTransforms(
    ItemToModify="((Item=(Type=Bone,Name=\"None\"),Transform=(Rotation=(X=0,Y=0,Z=0,W=1),Translation=(X=0,Y=0,Z=0),Scale3D=(X=1,Y=1,Z=1))),(Item=(Type=Bone,Name=\"deform_knee_out_l\"),Transform=(Rotation=(X=0,Y=0,Z=0,W=-1),Translation=(X=-8.6895,Y=7.008,Z=1.0244),Scale3D=(X=1,Y ...(truncated)"
    WeightMinimum=0
    Weight=@n1.Value
   )
   > &RigUnit_ModifyTransforms(
    ItemToModify="((Item=(Type=Bone,Name=\"None\"),Transform=(Rotation=(X=0,Y=0,Z=0,W=1),Translation=(X=0,Y=0,Z=0),Scale3D=(X=1,Y=1,Z=1))),(Item=(Type=Bone,Name=\"deform_knee_in_l\"),Transform=(Rotation=(X=0,Y=0,Z=0,W=1),Translation=(X=-16.1996,Y=7.4951,Z=-0.7789),Scale3D=(X=1, ...(truncated)"
    WeightMinimum=0
    Weight=@n1.Value
   )
   > &RigUnit_ModifyTransforms(
    ItemToModify="((Item=(Type=Bone,Name=\"None\"),Transform=(Rotation=(X=0,Y=0,Z=0,W=1),Translation=(X=0,Y=0,Z=0),Scale3D=(X=1,Y=1,Z=1))),(Item=(Type=Bone,Name=\"deform_groin_l\"),Transform=(Rotation=(X=-0.0003,Y=0.0533,Z=-0.0975,W=-0.9938),Translation=(X=-6.1415,Y=5.4527,Z=2. ...(truncated)"
    WeightMinimum=0
    Weight=RigVMVariableNode() {variable=L_thighUp}.Value
   )
   &RigUnit_RotationConstraintLocalSpaceOffset(
    AdvancedSettings=(RotationOrderForFilter=XZY)
    Child=(Type=Null,Name="dfrm_leg_spline_01_r_ctrl_null")
    Filter=(bX=True,bY=True,bZ=True)
    Parents=((Item=(Type=Bone,Name="dfrm_thigh_r"),Weight=0.9),(Item=(Type=Bone,Name="pelvis"),Weight=0.1))
    Weight=1
    ExecutePin=@RigVMFunction_Sequence.C
   )
   > &RigUnit_ParentConstraint(
    AdvancedSettings=(RotationOrderForFilter=XZY)
    Child=(Type=Null,Name="dfrm_leg_spline_03_r_ctrl_null")
    Parents=((Item=(Type=Bone,Name="dfrm_thigh_r"),Weight=1),(Item=(Type=Bone,Name="dfrm_calf_r"),Weight=0))
    Weight=1
   )
   > &RigUnit_ParentConstraint(
    AdvancedSettings=(RotationOrderForFilter=XZY)
    Child=(Type=Null,Name="dfrm_leg_spline_05_r_ctrl_null")
    Parents=((Item=(Type=Bone,Name="dfrm_thigh_r"),Weight=0.5),(Item=(Type=Bone,Name="dfrm_calf_r"),Weight=0.5))
    Weight=1
   )
   > &RigUnit_ParentConstraint(
    AdvancedSettings=(RotationOrderForFilter=XZY)
    Child=(Type=Null,Name="dfrm_leg_spline_07_r_ctrl_null")
    Parents=((Item=(Type=Bone,Name="dfrm_calf_r"),Weight=1),(Item=(Type=Bone,Name="dfrm_foot_r"),Weight=0))
    Weight=1
   )
   > &RigUnit_ParentConstraint(
    AdvancedSettings=(RotationOrderForFilter=XZY)
    Child=(Type=Null,Name="dfrm_leg_spline_09_r_ctrl_null")
    Parents=((Item=(Type=Bone,Name="dfrm_calf_r"),Weight=0),(Item=(Type=Bone,Name="dfrm_foot_r"),Weight=1))
    Weight=1
   )
   n2 = RigVMVariableNode() {variable=R_calfUp}
   > &RigUnit_ModifyTransforms(
    ItemToModify="((Item=(Type=Bone,Name=\"None\"),Transform=(Rotation=(X=0,Y=0,Z=0,W=1),Translation=(X=0,Y=0,Z=0),Scale3D=(X=1,Y=1,Z=1))),(Item=(Type=Null,Name=\"dfrm_leg_spline_05_r_ctrl_null\"),Transform=(Rotation=(X=0,Y=0,Z=-0.07,W=-0.9975),Translation=(X=-5.4837,Y=-17.2298 ...(truncated)"
    WeightMinimum=0
    Weight=@n2.Value
   )
   ItemArray_1 = RigVMDispatchNode(Value="((Type=Bone,Name=\"dfrm_leg_spline_01_r\"),(Type=Bone,Name=\"dfrm_leg_spline_02_r\"),(Type=Bone,Name=\"dfrm_leg_spline_03_r\"),(Type=Bone,Name=\"dfrm_leg_spline_04_r\"),(Type=Bone,Name=\"dfrm_leg_spline_05_r\"),(Type=Bone,Name=\"dfrm_leg_spline_06_r\"),(Type=Bone,Name ...(truncated)") {template="DISPATCH_RigVMDispatch_Constant(io Value)"}
   > SplineIK_1 = RigVMFunctionReferenceNode(
    Primary Axis=(X=-1,Y=0,Z=0)
    Samples Per Segment=16
    Secondary Spline Direction=(X=0,Y=0,Z=0)
    Spline Mode=BSpline
    Stretch=true
    Up Axis=(X=0,Y=0,Z=0)
    debug=false
    Bones=@ItemArray_1.Value
    Controls=RigVMVariableNode() {variable=R_legSpline_ctrls}.Value
   )
   > &RigUnit_ModifyTransforms(
    ItemToModify="((Item=(Type=Bone,Name=\"None\"),Transform=(Rotation=(X=0,Y=0,Z=0,W=1),Translation=(X=0,Y=0,Z=0),Scale3D=(X=1,Y=1,Z=1))),(Item=(Type=Bone,Name=\"deform_knee_out_r\"),Transform=(Rotation=(X=0,Y=0,Z=0.1736,W=-0.9848),Translation=(X=6.1203,Y=5.5811,Z=-2.6927),Sca ...(truncated)"
    WeightMinimum=0
    Weight=@n2.Value
   )
   > &RigUnit_ModifyTransforms(
    ItemToModify="((Item=(Type=Bone,Name=\"None\"),Transform=(Rotation=(X=0,Y=0,Z=0,W=1),Translation=(X=0,Y=0,Z=0),Scale3D=(X=1,Y=1,Z=1))),(Item=(Type=Bone,Name=\"deform_knee_in_r\"),Transform=(Rotation=(X=0,Y=0,Z=0.0872,W=-0.9962),Translation=(X=12.1858,Y=0.8326,Z=-2.6927),Sca ...(truncated)"
    WeightMinimum=0
    Weight=@n2.Value
   )
   > &RigUnit_ModifyTransforms(
    ItemToModify="((Item=(Type=Bone,Name=\"None\"),Transform=(Rotation=(X=0,Y=0,Z=0,W=1),Translation=(X=0,Y=0,Z=0),Scale3D=(X=1,Y=1,Z=1))),(Item=(Type=Bone,Name=\"deform_groin_r\"),Transform=(Rotation=(X=0,Y=0,Z=0,W=-1),Translation=(X=-4.7835,Y=3.3504,Z=-4.8542),Scale3D=(X=1,Y= ...(truncated)"
    WeightMinimum=0
    Weight=RigVMVariableNode() {variable=R_thighUp}.Value
   )
   RigVMFunction_MathBoolToFloat = &RigVMFunction_MathBoolToFloat(Value=RigVMVariableNode() {variable=isCrouching_CR}.Value)
   n3 = &RigVMFunction_AlphaInterp(
    InterpSpeedDecreasing=10
    InterpSpeedIncreasing=10
    OutRange=(Min=0,Max=1.4)
    Scale=1
    bClampResult=true
    Value=@RigVMFunction_MathBoolToFloat.Result
   )
   > &RigUnit_ModifyTransforms(
    ItemToModify="((Item=(Type=Bone,Name=\"deform_knee_in_l\"),Transform=(Rotation=(X=0,Y=0,Z=-0.1488,W=-0.9889),Translation=(X=0,Y=0,Z=0),Scale3D=(X=1,Y=1,Z=1))),(Item=(Type=Bone,Name=\"deform_knee_out_r\"),Transform=(Rotation=(X=0,Y=0,Z=0,W=-1),Translation=(X=-6.6573,Y=0.1965 ...(truncated)"
    WeightMinimum=0
    Weight=@n3.Result
   )
   RigVMFunctionReturnNode(ExecuteContext=@RigVMFunction_Sequence.D)
  ]}
 New Function_ContainedGraph {nodes [
   n1 = RigVMFunctionEntryNode()
   RigVMFunction_ControlFlowBranch = &RigVMFunction_ControlFlowBranch(Condition=@n1.Debug Controls, ExecuteContext=@n1.ExecuteContext)
   RigVMFunctionReturnNode(ExecuteContext=@RigVMFunction_ControlFlowBranch.Completed)
   RigUnit_GetTransform = &RigUnit_GetTransform(Item=(Type=Control,Name="thigh_r_ctrl"), bInitial=False)
   &RigUnit_SetTransform(
    Item=(Type=Bone,Name="thigh_r")
    Weight=1
    bInitial=False
    Value=@RigUnit_GetTransform.Transform
    ExecutePin=@RigVMFunction_ControlFlowBranch.True
   )
   > &RigUnit_SetTransform(
    Item=(Type=Bone,Name="thigh_l")
    Weight=1
    bInitial=False
    Value=&RigUnit_GetTransform(Item=(Type=Control,Name="thigh_l_ctrl"), bInitial=False).Transform
   )
   > &RigUnit_SetTransform(
    Item=(Type=Bone,Name="calf_l")
    Weight=1
    bInitial=False
    Value=&RigUnit_GetTransform(Item=(Type=Control,Name="calf_l_ctrl"), bInitial=False).Transform
   )
   > &RigUnit_SetTransform(
    Item=(Type=Bone,Name="calf_r")
    Weight=1
    bInitial=False
    Value=&RigUnit_GetTransform(Item=(Type=Control,Name="calf_r_ctrl"), bInitial=False).Transform
   )
   > &RigUnit_SetTransform(
    Item=(Type=Bone,Name="upperarm_l")
    Weight=1
    bInitial=False
    Value=&RigUnit_GetTransform(Item=(Type=Control,Name="upperarm_l_ctrl"), bInitial=False).Transform
   )
   > &RigUnit_SetTransform(
    Item=(Type=Bone,Name="upperarm_r")
    Weight=1
    bInitial=False
    Value=&RigUnit_GetTransform(Item=(Type=Control,Name="upperarm_r_ctrl"), bInitial=False).Transform
   )
   > &RigUnit_SetTransform(
    Item=(Type=Bone,Name="lowerarm_l")
    Weight=1
    bInitial=False
    Value=&RigUnit_GetTransform(Item=(Type=Control,Name="lowerarm_l_ctrl"), bInitial=False).Transform
   )
   > &RigUnit_SetTransform(
    Item=(Type=Bone,Name="lowerarm_r")
    Weight=1
    bInitial=False
    Value=&RigUnit_GetTransform(Item=(Type=Control,Name="lowerarm_r_ctrl"), bInitial=False).Transform
   )
   &RigUnit_SetTransform(
    ExecutePin=()
    Item=(Type=Bone,Name="spine_05")
    Weight=1
    bInitial=False
    Value=&RigUnit_GetTransform(Item=(Type=Control,Name="spine_05_ctrl"), bInitial=False).Transform
   )
  ]}
 New Function_ContainedGraph {nodes [
   n1 = RigVMFunctionEntryNode()
   RigVMFunction_MathQuaternionToEuler = &RigVMFunction_MathQuaternionToEuler(
    RotationOrder=XYZ
    Value=&RigUnit_GetRelativeTransformForItem(
     Child=(Type=Bone,Name="hand_l")
     Parent=(Type=Null,Name="lowerarm_l")
     bChildInitial=False
     bParentInitial=False
     Child.Name=@n1.Child
     Parent.Name=@n1.Parent
    ).RelativeTransform.Rotation
   )
   Remap = &RigVMFunction_MathDoubleRemap(
    TargetMaximum=1
    TargetMinimum=0
    Value=@RigVMFunction_MathQuaternionToEuler.Result.Z
    SourceMaximum=@n1.Source Maximum
    SourceMinimum=@n1.Source Minimum
   )
   > RigVMFunctionReturnNode(Delta Rotation=@Remap.Result)
  ]}
 RigVMFunctionLibrary {nodes [
   Control Setup = RigVMCollapseNode(Debug Controls=False, ExecuteContext=())
   Delta Rotation Z = RigVMCollapseNode(ExecuteContext=())
   arm_Spline = RigVMCollapseNode(ExecuteContext=())
   leg_Spline = RigVMCollapseNode(ExecuteContext=())
  ]}
 RigVMFunction_Sequence_1_ContainedGraph {nodes [
   n1 = RigVMFunctionEntryNode()
   RigVMFunction_Sequence = &RigVMFunction_Sequence(ExecuteContext=@n1.ExecuteContext)
   n2 = &RigVMFunction_Sequence(ExecuteContext=@RigVMFunction_Sequence.B)
   RigVMFunctionReturnNode(A=@RigVMFunction_Sequence.A, B=@n2.A, C=@n2.B)
  ]}
 RigVMFunction_Sequence_1_ContainedGraph {nodes [
   n1 = RigVMFunctionEntryNode()
   RigVMFunction_Sequence = &RigVMFunction_Sequence(ExecuteContext=@n1.ExecuteContext)
   n2 = &RigVMFunction_Sequence(ExecuteContext=@RigVMFunction_Sequence.B)
   n3 = &RigVMFunction_Sequence(ExecuteContext=@n2.B)
   RigVMFunctionReturnNode(A=@RigVMFunction_Sequence.A, B=@n2.A, C=@n3.A, D=@n3.B)
  ]}
 RigVMModel {nodes [
   RigUnit_BeginExecution = &RigUnit_BeginExecution()
   Control Setup = RigVMFunctionReferenceNode(Debug Controls=false, ExecuteContext=@RigUnit_BeginExecution.ExecutePin)
   RigVMFunction_Sequence = &RigVMFunction_Sequence(ExecuteContext=@"Control Setup".ExecuteContext)
   arm_Spline = RigVMFunctionReferenceNode(ExecuteContext=@RigVMFunction_Sequence.A)
   leg_Spline = RigVMFunctionReferenceNode(ExecuteContext=@RigVMFunction_Sequence.B)
  ]}
 RigVMModel Contruction Event Graph {nodes [
   n1 = &RigUnit_PrepareForExecution()
   RigVMFunction_Sequence = &RigVMFunction_Sequence(ExecuteContext=@n1.ExecutePin)
   ItemArray = RigVMDispatchNode(Value=((Type=Control,Name="dfrm_arm_spline_01_l_ctrl"),(Type=Control,Name="dfrm_arm_spline_03_l_ctrl"),(Type=Control,Name="dfrm_arm_spline_05_l_ctrl"),(Type=Control,Name="dfrm_arm_spline_07_l_ctrl"),(Type=Control,Name="dfrm_arm_spline_09_l_ctrl"))) {template="DISPATCH_RigVMDispatch_Constant(io Value)"}
   RigVMVariableNode(Value=@ItemArray.Value, ExecuteContext=@RigVMFunction_Sequence.A) {variable=L_armSpline_ctrls}
   ItemArray_1 = RigVMDispatchNode(Value=((Type=Control,Name="dfrm_leg_spline_01_l_ctrl"),(Type=Control,Name="dfrm_leg_spline_03_l_ctrl"),(Type=Control,Name="dfrm_leg_spline_05_l_ctrl"),(Type=Control,Name="dfrm_leg_spline_07_l_ctrl"),(Type=Control,Name="dfrm_leg_spline_09_l_ctrl"))) {template="DISPATCH_RigVMDispatch_Constant(io Value)"}
   > RigVMVariableNode(Value=@ItemArray_1.Value) {variable=L_legSpline_ctrls}
   ItemArray_3 = RigVMDispatchNode(Value=((Type=Control,Name="dfrm_arm_spline_01_r_ctrl"),(Type=Control,Name="dfrm_arm_spline_03_r_ctrl"),(Type=Control,Name="dfrm_arm_spline_05_r_ctrl"),(Type=Control,Name="dfrm_arm_spline_07_r_ctrl"),(Type=Control,Name="dfrm_arm_spline_09_r_ctrl"))) {template="DISPATCH_RigVMDispatch_Constant(io Value)"}
   > RigVMVariableNode(Value=@ItemArray_3.Value) {variable=R_armSpline_ctrls}
   ItemArray_2 = RigVMDispatchNode(Value=((Type=Control,Name="dfrm_leg_spline_01_r_ctrl"),(Type=Control,Name="dfrm_leg_spline_03_r_ctrl"),(Type=Control,Name="dfrm_leg_spline_05_r_ctrl"),(Type=Control,Name="dfrm_leg_spline_07_r_ctrl"),(Type=Control,Name="dfrm_leg_spline_09_r_ctrl"))) {template="DISPATCH_RigVMDispatch_Constant(io Value)"}
   > RigVMVariableNode(Value=@ItemArray_2.Value) {variable=R_legSpline_ctrls}
  ]}
]
hierarchy [
 root {loc=<0 0 0>, rot=<0 0 0 1>}
 pelvis {parent=root, loc=<0 -0.7071 93.1191>, rot=<0.0127 0.707 -0.0127 -0.707>}
 spine_01 {parent=pelvis, loc=<6.7494 -0.6388 0>, rot=<0 0 0.034 -0.9994>}
 spine_02 {parent=spine_01, loc=<7.2518 -0.035 0>, rot=<0 0 0.0238 -0.9997>}
 spine_03 {parent=spine_02, loc=<6.0104 0 0>, rot=<0 0 0.0149 -0.9999>}
 spine_04 {parent=spine_03, loc=<8.1024 0 0>, rot=<0 0 -0.0132 -0.9999>}
 spine_05 {parent=spine_04, loc=<10.0146 0 0>, rot=<0 0 -0.0015 -1>}
 neck_01 {parent=spine_05, loc=<11.3138 0.1529 0>, rot=<0 0 -0.2003 -0.9797>}
 neck_02 {parent=neck_01, loc=<4.5627 0 0>, rot=<0 0 0.0504 -0.9987>}
 head {parent=neck_02, loc=<5.3914 0 0>, rot=<0 0 0.115 -0.9934>}
 attach_cape {parent=spine_05, loc=<8.7408 -11.1475 0.2058>, rot=<0.0616 -0.7044 -0.0616 -0.7044>}
 cape {parent=attach_cape, type=Null, loc=<0 0 0>, rot=<0 0 0 1>}
 attach_backpack {parent=spine_05, loc=<-0.1653 -12.8276 -0.0149>, rot=<-0.0093 -0.7071 0.0092 -0.707>}
 Backpack_BR {parent=attach_backpack, type=Null, loc=<0 0 0>, rot=<0 0 0 1>}
 clavicle_l {parent=spine_05, loc=<4.3575 -0.1466 -2.3223>, rot=<0.0465 -0.7289 -0.009 -0.6829>}
 upperarm_l {parent=clavicle_l, loc=<15.5111 0 0>, rot=<0.0226 -0.3234 -0.0088 -0.9459>}
 lowerarm_l {parent=upperarm_l, loc=<26.6514 0 0>, rot=<0.0391 -0.0099 -0.2453 -0.9686>}
 hand_l {parent=lowerarm_l, loc=<22.3491 0 0>, rot=<-0.7081 -0.0507 0.0975 -0.6975>}
 thumb_01_l {parent=hand_l, loc=<1.3118 -0.3191 -2.459>, rot=<0.6546 -0.4042 0.0664 -0.6354>}
 thumb_02_l {parent=thumb_01_l, loc=<5.9336 -0.0448 0.0401>, rot=<-0.0037 0.0213 0.298 -0.9543>}
 thumb_03_l {parent=thumb_02_l, loc=<2.2223 0.3682 -0.0907>, rot=<0.0118 -0.013 0.0066 -0.9998>}
 index_metacarpal_l {parent=hand_l, loc=<1.6535 0.0183 -2.96>, rot=<0.0158 -0.0124 -0.0173 -0.9996>}
 index_01_l {parent=index_metacarpal_l, loc=<8.3008 0.0357 -0.0524>, rot=<-0.0044 0.0485 0.0924 -0.9945>}
 index_02_l {parent=index_01_l, loc=<3.3587 -0.3004 0.0452>, rot=<-0.002 0.0027 0.3225 -0.9466>}
 index_03_l {parent=index_02_l, loc=<2.5578 -0.123 -0.0252>, rot=<-0.002 0.0067 0.2558 -0.9667>}
 middle_metacarpal_l {parent=hand_l, loc=<2.1856 -0.1725 -0.5984>, rot=<0.006 0.0168 -0.0329 -0.9993>}
 middle_01_l {parent=middle_metacarpal_l, loc=<7.3275 -0.2447 0.0129>, rot=<-0.0208 0.0042 0.0858 -0.9961>}
 middle_02_l {parent=middle_01_l, loc=<4.0123 0.1401 0.1175>, rot=<-0.0218 0.0258 0.3512 -0.9357>}
 middle_03_l {parent=middle_02_l, loc=<2.6077 0.0534 0.1373>, rot=<0.0082 0.0188 0.1708 -0.9851>}
 ring_metacarpal_l {parent=hand_l, loc=<2.2074 -0.3625 1.3638>, rot=<-0.0284 0.0722 -0.0458 -0.9959>}
 ring_01_l {parent=ring_metacarpal_l, loc=<7.0295 -0.6197 -0.1446>, rot=<-0.0489 -0.0252 0.1892 -0.9804>}
 ring_02_l {parent=ring_01_l, loc=<3.6174 0.4647 -0.1027>, rot=<0 0 0.2806 -0.9598>}
 ring_03_l {parent=ring_02_l, loc=<2.4152 0.114 0.0199>, rot=<0 0 0.1189 -0.9929>}
 pinky_metacarpal_l {parent=hand_l, loc=<1.7616 -0.9753 3.3814>, rot=<-0.0397 0.0814 -0.069 -0.9935>}
 pinky_01_l {parent=pinky_metacarpal_l, loc=<7.1709 -0.4831 -0.1094>, rot=<-0.0654 -0.0035 0.2242 -0.9723>}
 pinky_02_l {parent=pinky_01_l, loc=<2.8274 0.0893 -0.1043>, rot=<0.0037 0.0011 0.2854 -0.9584>}
 pinky_03_l {parent=pinky_02_l, loc=<2.2046 0.0307 -0.0078>, rot=<-0.0002 0.0098 0.1445 -0.9895>}
 deform_wrist_dn_base_l {parent=hand_l, loc=<0 -2.2342 0>, rot=<0 0 0 -1>}
 deform_wrist_dn_l {parent=deform_wrist_dn_base_l, loc=<0 0 0>, rot=<0 0 0 1>}
 deform_wrist_up_base_l {parent=hand_l, loc=<0 2.2682 0>, rot=<0 0 0 -1>}
 deform_wrist_up_l {parent=deform_wrist_up_base_l, loc=<0 0 0>, rot=<0 0 0 1>}
 dfrm_arm_spline_09_l_ctrl_null {parent=hand_l, type=Null, loc=<0 0 0>, rot=<0.7099 0.004 -0.1432 -0.6896>}
 dfrm_arm_spline_09_l_ctrl {controlType=EULER_TRANSFORM, parent=dfrm_arm_spline_09_l_ctrl_null, type=Control, loc=<0 0 0>, rot=<0 0 0 1>}
 lowerarm_twist_02_l {parent=lowerarm_l, loc=<9.6199 0 0>, rot=<0 0 0 -1>}
 lowerarm_twist_01_l {parent=lowerarm_l, loc=<17.8549 0 0>, rot=<0 0 0 -1>}
 deform_elbow_in_base_l {parent=lowerarm_l, loc=<0.2112 5.259 0.8274>, rot=<0 0 0 1>}
 deform_elbow_in_l {parent=deform_elbow_in_base_l, loc=<0 0 0>, rot=<0 0 0 1>}
 deform_elbow_out_base_l {parent=lowerarm_l, loc=<0 -4 0>, rot=<0 0 0 1>}
 deform_elbow_out_l {parent=deform_elbow_out_base_l, loc=<0 0 0>, rot=<0 0 0 1>}
 upperarm_twist_01_l {parent=upperarm_l, loc=<0.0041 0 0>, rot=<0 0 0 -1>}
 deform_upperarm_bk_base_l {parent=upperarm_twist_01_l, loc=<1.0813 -7.2193 -1.9453>, rot=<0 0 0 1>}
 deform_upperarm_bk_l {parent=deform_upperarm_bk_base_l, loc=<0 0 0>, rot=<0 0 0 1>}
 deform_upperarm_fr_base_l {parent=upperarm_twist_01_l, loc=<-0.3691 7.0506 -2.765>, rot=<0 0 0 1>}
 deform_upperarm_fr_l {parent=deform_upperarm_fr_base_l, loc=<0 0 0>, rot=<0 0 0 1>}
 deform_upperarm_up_base_l {parent=upperarm_twist_01_l, loc=<-3.2216 -1.1391 3.3311>, rot=<0 0 0 1>}
 deform_upperarm_up_l {parent=deform_upperarm_up_base_l, loc=<0 0 0>, rot=<0 0 0 1>}
 upperarm_twist_02_l {parent=upperarm_l, loc=<13.7941 0 0>, rot=<0 0 0 -1>}
 dfrm_arm_spline_MSTR_l {parent=clavicle_l, loc=<15.5111 0 0>, rot=<0.0008 0.0336 -0.0252 -0.9991>}
 dfrm_arm_spline_01_l {parent=dfrm_arm_spline_MSTR_l, loc=<0 0 0>, rot=<-0.0286 -0.3471 0.0771 -0.9342>}
 dfrm_arm_spline_02_l {parent=dfrm_arm_spline_01_l, loc=<6.161 0 0>, rot=<0.102 -0.0013 -0.0127 -0.9947>}
 dfrm_arm_spline_03_l {parent=dfrm_arm_spline_02_l, loc=<6.161 0 0>, rot=<0 0 -0.0283 -0.9996>}
 dfrm_arm_spline_04_l {parent=dfrm_arm_spline_03_l, loc=<6.1592 0 0>, rot=<0 0 -0.0503 -0.9987>}
 dfrm_arm_spline_05_l {parent=dfrm_arm_spline_04_l, loc=<6.1568 0 0>, rot=<0 0 0 1>}
 dfrm_arm_spline_06_l {parent=dfrm_arm_spline_05_l, loc=<6.0706 0.9956 0>, rot=<0 0 0 1>}
 dfrm_arm_spline_07_l {parent=dfrm_arm_spline_06_l, loc=<5.8031 2.0501 0>, rot=<0 0 0 1>}
 dfrm_arm_spline_08_l {parent=dfrm_arm_spline_07_l, loc=<5.4954 2.7819 0>, rot=<0 0 0 1>}
 dfrm_arm_spline_09_l {parent=dfrm_arm_spline_08_l, loc=<5.3018 3.1358 0>, rot=<0 0 -0.2639 -0.9646>}
 dfrm_arm_spline_01_l_ctrl_null {parent=clavicle_l, type=Null, loc=<15.5111 0 0>, rot=<0.0217 0.3161 -0.0528 0.947>}
 dfrm_arm_spline_01_l_ctrl {controlType=EULER_TRANSFORM, parent=dfrm_arm_spline_01_l_ctrl_null, type=Control, loc=<0 0 0>, rot=<0 0 0 1>}
 clavicle_r {parent=spine_05, loc=<4.3575 -0.1466 2.3223>, rot=<0.7289 0.0465 -0.6829 0.009>}
 upperarm_r {parent=clavicle_r, loc=<-15.5111 0 0>, rot=<0.0226 -0.3234 -0.0088 -0.9459>}
 lowerarm_r {parent=upperarm_r, loc=<-26.6514 0 0>, rot=<0.0391 -0.0099 -0.2453 -0.9686>}
 hand_r {parent=lowerarm_r, loc=<-22.3491 0 0>, rot=<-0.7081 -0.0507 0.0975 -0.6975>}
 thumb_01_r {parent=hand_r, loc=<-1.3118 0.3191 2.459>, rot=<0.6546 -0.4042 0.0664 -0.6354>}
 thumb_02_r {parent=thumb_01_r, loc=<-5.9336 0.0448 -0.0401>, rot=<-0.0037 0.0213 0.298 -0.9543>}
 thumb_03_r {parent=thumb_02_r, loc=<-2.2223 -0.3682 0.0907>, rot=<0.0118 -0.013 0.0066 -0.9998>}
 index_metacarpal_r {parent=hand_r, loc=<-1.6535 -0.0183 2.96>, rot=<0.0158 -0.0124 -0.0173 -0.9996>}
 index_01_r {parent=index_metacarpal_r, loc=<-8.3008 -0.0357 0.0524>, rot=<-0.0044 0.0485 0.0924 -0.9945>}
 index_02_r {parent=index_01_r, loc=<-3.3587 0.3004 -0.0452>, rot=<-0.002 0.0027 0.3225 -0.9466>}
 index_03_r {parent=index_02_r, loc=<-2.5578 0.123 0.0252>, rot=<-0.002 0.0067 0.2558 -0.9667>}
 middle_metacarpal_r {parent=hand_r, loc=<-2.1856 0.1725 0.5984>, rot=<0.006 0.0168 -0.0329 -0.9993>}
 middle_01_r {parent=middle_metacarpal_r, loc=<-7.3275 0.2447 -0.0129>, rot=<-0.0208 0.0042 0.0858 -0.9961>}
 middle_02_r {parent=middle_01_r, loc=<-4.0123 -0.1401 -0.1175>, rot=<-0.0218 0.0258 0.3512 -0.9357>}
 middle_03_r {parent=middle_02_r, loc=<-2.6077 -0.0534 -0.1373>, rot=<0.0082 0.0188 0.1708 -0.9851>}
 ring_metacarpal_r {parent=hand_r, loc=<-2.2074 0.3625 -1.3638>, rot=<-0.0284 0.0722 -0.0458 -0.9959>}
 ring_01_r {parent=ring_metacarpal_r, loc=<-7.0295 0.6197 0.1446>, rot=<-0.0489 -0.0252 0.1892 -0.9804>}
 ring_02_r {parent=ring_01_r, loc=<-3.6174 -0.4647 0.1027>, rot=<0 0 0.2806 -0.9598>}
 ring_03_r {parent=ring_02_r, loc=<-2.4152 -0.114 -0.0199>, rot=<0 0 0.1189 -0.9929>}
 pinky_metacarpal_r {parent=hand_r, loc=<-1.7616 0.9753 -3.3814>, rot=<-0.0397 0.0814 -0.069 -0.9935>}
 pinky_01_r {parent=pinky_metacarpal_r, loc=<-7.1709 0.4831 0.1094>, rot=<-0.0654 -0.0035 0.2242 -0.9723>}
 pinky_02_r {parent=pinky_01_r, loc=<-2.8274 -0.0893 0.1043>, rot=<0.0037 0.0011 0.2854 -0.9584>}
 pinky_03_r {parent=pinky_02_r, loc=<-2.2046 -0.0307 0.0078>, rot=<-0.0002 0.0098 0.1445 -0.9895>}
 deform_wrist_dn_base_r {parent=hand_r, loc=<-0.0004 2.2345 -0.0001>, rot=<0 0 0 -1>}
 deform_wrist_dn_r {parent=deform_wrist_dn_base_r, loc=<0 0 0>, rot=<0 0 0 1>}
 deform_wrist_up_base_r {parent=hand_r, loc=<0.0003 -2.2684 0.0001>, rot=<0 0 0 -1>}
 deform_wrist_up_r {parent=deform_wrist_up_base_r, loc=<0 0 0>, rot=<0 0 0 1>}
 dfrm_arm_spline_09_r_ctrl_null {parent=hand_r, type=Null, loc=<0.0004 -0.0002 0.0001>, rot=<0.7099 0.004 -0.1432 -0.6896>}
 dfrm_arm_spline_09_r_ctrl {controlType=EULER_TRANSFORM, parent=dfrm_arm_spline_09_r_ctrl_null, type=Control, loc=<0 0 0>, rot=<0 0 0 1>}
 lowerarm_twist_02_r {parent=lowerarm_r, loc=<-9.6199 0 0>, rot=<0 0 0 -1>}
 lowerarm_twist_01_r {parent=lowerarm_r, loc=<-17.8549 0 0>, rot=<0 0 0 -1>}
 deform_elbow_in_base_r {parent=lowerarm_r, loc=<-0.2111 -5.2591 -0.8275>, rot=<0 0 0 -1>}
 deform_elbow_in_r {parent=deform_elbow_in_base_r, loc=<-0.0001 0 0.0002>, rot=<0 0 0 1>}
 deform_elbow_out_base_r {parent=lowerarm_r, loc=<0 4 -0.0001>, rot=<0 0 0 -1>}
 deform_elbow_out_r {parent=deform_elbow_out_base_r, loc=<0 0 0>, rot=<0 0 0 1>}
 upperarm_twist_01_r {parent=upperarm_r, loc=<-0.0041 0 0>, rot=<0 0 0 -1>}
 deform_upperarm_bk_base_r {parent=upperarm_twist_01_r, loc=<-1.0814 7.2192 1.9455>, rot=<0 0 0 1>}
 deform_upperarm_bk_r {parent=deform_upperarm_bk_base_r, loc=<0 0 0>, rot=<0 0 0 1>}
 deform_upperarm_fr_base_r {parent=upperarm_twist_01_r, loc=<0.3689 -7.0506 2.7652>, rot=<0 0 0 1>}
 deform_upperarm_fr_r {parent=deform_upperarm_fr_base_r, loc=<0 0 0>, rot=<0 0 0 1>}
 deform_upperarm_up_base_r {parent=upperarm_twist_01_r, loc=<3.2214 1.1391 -3.3308>, rot=<0 0 0 1>}
 deform_upperarm_up_r {parent=deform_upperarm_up_base_r, loc=<0 0 0>, rot=<0 0 0 -1>}
 upperarm_twist_02_r {parent=upperarm_r, loc=<-13.7941 0 0>, rot=<0 0 0 -1>}
 dfrm_arm_spline_MSTR_r {parent=clavicle_r, loc=<-15.5111 0 0.0001>, rot=<0.0008 0.0336 -0.0252 -0.9991>}
 dfrm_arm_spline_01_r {parent=dfrm_arm_spline_MSTR_r, loc=<0 0 0>, rot=<-0.0286 -0.3471 0.0771 -0.9342>}
 dfrm_arm_spline_02_r {parent=dfrm_arm_spline_01_r, loc=<-6.1608 0 -0.0003>, rot=<0 0 0 1>}
 dfrm_arm_spline_03_r {parent=dfrm_arm_spline_02_r, loc=<-6.1595 -0.1546 0.0325>, rot=<0 0 0 1>}
 dfrm_arm_spline_04_r {parent=dfrm_arm_spline_03_r, loc=<-6.1381 -0.4957 0.1024>, rot=<0 0 0 1>}
 dfrm_arm_spline_05_r {parent=dfrm_arm_spline_04_r, loc=<-6.0543 -1.0965 0.2275>, rot=<0.1005 -0.0175 -0.171 -0.98>}
 dfrm_arm_spline_06_r {parent=dfrm_arm_spline_05_r, loc=<-6.1515 0 -0.0002>, rot=<0 0 0 1>}
 dfrm_arm_spline_07_r {parent=dfrm_arm_spline_06_r, loc=<-6.0583 -1.0839 -0.0002>, rot=<0 0 0 1>}
 dfrm_arm_spline_08_r {parent=dfrm_arm_spline_07_r, loc=<-5.8733 -1.8559 0.0002>, rot=<0 0 0 1>}
 dfrm_arm_spline_09_r {parent=dfrm_arm_spline_08_r, loc=<-5.739 -2.2365 -0.0004>, rot=<0 0 -0.1847 -0.9828>}
 dfrm_arm_spline_01_r_ctrl_null {parent=clavicle_r, type=Null, loc=<-15.5111 0 0.0001>, rot=<0.0217 0.3161 -0.0528 0.947>}
 dfrm_arm_spline_01_r_ctrl {controlType=EULER_TRANSFORM, parent=dfrm_arm_spline_01_r_ctrl_null, type=Control, loc=<0 0 0>, rot=<0 0 0 1>}
 deform_pec_aim_base_l {parent=spine_04, loc=<7.8732 12.6388 -2.4123>, rot=<0.0102 -0.4602 0.3979 -0.7936>}
 deform_pec_aim_l {parent=deform_pec_aim_base_l, loc=<0 0 0>, rot=<0 0 0 1>}
 deform_pec_base_l {parent=deform_pec_aim_l, loc=<10.766 0 0>, rot=<0 0 0 1>}
 deform_pec_l {parent=deform_pec_base_l, loc=<0.055 0 0>, rot=<0 0 0 1>}
 deform_pec_aim_base_r {parent=spine_04, loc=<7.8727 12.6388 2.4123>, rot=<-0.4602 -0.0102 0.7936 0.3979>}
 deform_pec_aim_r {parent=deform_pec_aim_base_r, loc=<0 0 0>, rot=<0 0 0 1>}
 deform_pec_base_r {parent=deform_pec_aim_r, loc=<-10.7662 -0.0006 -0.0007>, rot=<0 0 0 1>}
 deform_pec_r {parent=deform_pec_base_r, loc=<0 0 0>, rot=<0 0 0 1>}
 thigh_l {parent=pelvis, loc=<-2.5693 -0.0109 -7.7501>, rot=<0.0012 0.0572 0.0155 -0.9982>}
 calf_l {parent=thigh_l, loc=<-39.0484 0 0>, rot=<0.0006 -0.0161 -0.0391 -0.9991>}
 foot_l {parent=calf_l, loc=<-42.561 0.1077 0.4292>, rot=<0.0418 -0.0455 0.0343 -0.9975>}
 ball_l {parent=foot_l, loc=<-6.1599 13.8264 0.6806>, rot=<0.0001 -0.0001 -0.7063 -0.7079>}
 calf_twist_02_l {parent=calf_l, loc=<-18.8064 -0.0967 0.0397>, rot=<0.0237 0.0095 -0.0035 -0.9997>}
 calf_twist_01_l {parent=calf_l, loc=<-35.813 -0.0967 0.0397>, rot=<0.0237 0.0095 -0.0035 -0.9997>}
 deform_knee_in_base_l {parent=calf_l, loc=<1.298 -5.2957 -0.0496>, rot=<0 0 0 -1>}
 deform_knee_in_l {parent=deform_knee_in_base_l, loc=<0 0 0>, rot=<0 0 0 1>}
 deform_knee_out_base_l {parent=calf_l, loc=<0 4 0>, rot=<0 0 0 -1>}
 deform_knee_out_l {parent=deform_knee_out_base_l, loc=<0 0 0>, rot=<0 0 0 1>}
 thigh_twist_01_l {parent=thigh_l, loc=<-21.6303 0 0>, rot=<0.0013 0 0 -1>}
 deform_groin_base_l {parent=thigh_l, loc=<-2.1908 8.1142 -1.7442>, rot=<0 0 0 -1>}
 deform_groin_l {parent=deform_groin_base_l, loc=<0 0 0>, rot=<0 0 0 1>}
 calf_l_null {parent=thigh_l, type=Null, loc=<-42.6652 33.0624 15.8309>, rot=<0.2961 0.0143 0.1864 0.9367>}
 thigh_r {parent=pelvis, loc=<-2.5693 -0.0109 7.7501>, rot=<0.0572 -0.0012 0.9982 0.0155>}
 calf_r {parent=thigh_r, loc=<39.0484 0 0>, rot=<0.0006 -0.0161 -0.0391 -0.9991>}
 foot_r {parent=calf_r, loc=<42.561 -0.1077 -0.4292>, rot=<0.0418 -0.0455 0.0343 -0.9975>}
 ball_r {parent=foot_r, loc=<6.1599 -13.8264 -0.6806>, rot=<0.0001 -0.0001 -0.7063 -0.7079>}
 calf_twist_02_r {parent=calf_r, loc=<18.8064 0.0967 -0.0397>, rot=<0.0237 0.0095 -0.0035 -0.9997>}
 calf_twist_01_r {parent=calf_r, loc=<35.813 0.0967 -0.0397>, rot=<0.0237 0.0095 -0.0035 -0.9997>}
 deform_knee_in_base_r {parent=calf_r, loc=<-1.298 5.2957 0.0496>, rot=<0 0 0 -1>}
 deform_knee_in_r {parent=deform_knee_in_base_r, loc=<0 0 0>, rot=<0 0 0 1>}
 deform_knee_out_base_r {parent=calf_r, loc=<0 -4 0>, rot=<0 0 0 -1>}
 deform_knee_out_r {parent=deform_knee_out_base_r, loc=<0 0 0>, rot=<0 0 0 1>}
 thigh_twist_01_r {parent=thigh_r, loc=<21.6303 0 0>, rot=<0.0013 0 0 -1>}
 deform_groin_base_r {parent=thigh_r, loc=<2.1908 -8.1142 1.7442>, rot=<0 0 0 -1>}
 deform_groin_r {parent=deform_groin_base_r, loc=<0 0 0>, rot=<0 0 0 1>}
 calf_r_null {parent=thigh_r, type=Null, loc=<39.0484 0 0>, rot=<0.0006 -0.0161 -0.0391 -0.9991>}
 deform_glute_base_l {parent=pelvis, loc=<-8.8064 -8.8222 -8.4287>, rot=<0.0275 0.0501 -0.4472 -0.8926>}
 deform_glute_l {parent=deform_glute_base_l, loc=<0 0 0>, rot=<0 0 0 1>}
 deform_glute_base_r {parent=pelvis, loc=<-8.8062 -8.8222 8.4287>, rot=<0.0501 -0.0275 0.8926 -0.4472>}
 deform_glute_r {parent=deform_glute_base_r, loc=<0 0 0>, rot=<0 0 0 1>}
 dfrm_leg_spline_MSTR_l {parent=pelvis, loc=<-14.5412 -1.3407 -11.2>, rot=<-0.0448 0.0002 0.9985 0.0318>}
 dfrm_leg_spline_01_l {parent=dfrm_leg_spline_MSTR_l, loc=<0 0 0>, rot=<0 0 0 1>}
 dfrm_leg_spline_02_l {parent=dfrm_leg_spline_01_l, loc=<8.6756 0 0>, rot=<-0.0001 0.0023 -0.004 -1>}
 dfrm_leg_spline_03_l {parent=dfrm_leg_spline_02_l, loc=<8.6755 0 0>, rot=<-0.0001 0.0044 -0.0077 -1>}
 dfrm_leg_spline_04_l {parent=dfrm_leg_spline_03_l, loc=<8.6755 0 0>, rot=<0 0.0063 -0.011 -0.9999>}
 dfrm_leg_spline_05_l {parent=dfrm_leg_spline_04_l, loc=<8.6753 0 0>, rot=<0.0002 0.0068 -0.012 -0.9999>}
 dfrm_leg_spline_06_l {parent=dfrm_leg_spline_05_l, loc=<8.6754 0 0>, rot=<0.0003 0.0051 -0.0089 -0.9999>}
 dfrm_leg_spline_07_l {parent=dfrm_leg_spline_06_l, loc=<8.6755 0 0>, rot=<0.0002 0.0032 -0.0056 -1>}
 dfrm_leg_spline_08_l {parent=dfrm_leg_spline_07_l, loc=<8.6756 0 0>, rot=<0 0 0 1>}
 dfrm_leg_spline_09_l {parent=dfrm_leg_spline_08_l, loc=<8.6754 0.0475 0.0271>, rot=<0.0001 0.0016 -0.0027 -1>}
 dfrm_leg_spline_MSTR_r {parent=pelvis, loc=<-14.5412 -1.3407 11.2>, rot=<-0.0002 -0.0448 0.0318 -0.9985>}
 dfrm_leg_spline_01_r {parent=dfrm_leg_spline_MSTR_r, loc=<0 0 0>, rot=<0 0 0 1>}
 dfrm_leg_spline_02_r {parent=dfrm_leg_spline_01_r, loc=<-8.6756 0 0>, rot=<0 0 0 1>}
 dfrm_leg_spline_03_r {parent=dfrm_leg_spline_02_r, loc=<-8.6751 -0.07 -0.0399>, rot=<0 0 0 1>}
 dfrm_leg_spline_04_r {parent=dfrm_leg_spline_03_r, loc=<-8.6723 -0.2037 -0.1161>, rot=<-0.0001 0.013 -0.0228 -0.9997>}
 dfrm_leg_spline_05_r {parent=dfrm_leg_spline_04_r, loc=<-8.6754 0 0>, rot=<0.0002 0.0068 -0.012 -0.9999>}
 dfrm_leg_spline_06_r {parent=dfrm_leg_spline_05_r, loc=<-8.6754 0 0>, rot=<0.0003 0.0051 -0.0089 -0.9999>}
 dfrm_leg_spline_07_r {parent=dfrm_leg_spline_06_r, loc=<-8.6755 0 -0.0001>, rot=<0.0002 0.0032 -0.0056 -1>}
 dfrm_leg_spline_08_r {parent=dfrm_leg_spline_07_r, loc=<-8.6756 0 0.0001>, rot=<0.0001 0.0016 -0.0027 -1>}
 dfrm_leg_spline_09_r {parent=dfrm_leg_spline_08_r, loc=<-8.6756 0 -0.0001>, rot=<0 0 0 1>}
 dfrm_arm_spline_03_l_ctrl_null {parent=pelvis, type=Null, loc=<33.0446 -6.6985 -26.9894>, rot=<-0.0226 0.9118 0.0276 0.409>}
 dfrm_arm_spline_03_l_ctrl {controlType=EULER_TRANSFORM, parent=dfrm_arm_spline_03_l_ctrl_null, type=Control, loc=<0 0 0>, rot=<0 0 0 1>}
 dfrm_arm_spline_05_l_ctrl_null {parent=pelvis, type=Null, loc=<24.843 -6.3118 -36.1483>, rot=<0.0972 0.9069 0.081 0.4019>}
 dfrm_arm_spline_05_l_ctrl {controlType=EULER_TRANSFORM, parent=dfrm_arm_spline_05_l_ctrl_null, type=Control, loc=<0 0 0>, rot=<0 0 0 1>}
 dfrm_arm_spline_07_l_ctrl_null {parent=pelvis, type=Null, loc=<16.9277 -2.3151 -44.6121>, rot=<0.2343 0.8815 0.1413 0.3848>}
 dfrm_arm_spline_07_l_ctrl {controlType=EULER_TRANSFORM, parent=dfrm_arm_spline_07_l_ctrl_null, type=Control, loc=<0 0 0>, rot=<0 0 0 1>}
 dfrm_thigh_r {parent=pelvis, loc=<-14.541 -1.3407 11.2>, rot=<0.0397 -0.0004 0.9989 0.0229>}
 dfrm_calf_r {parent=dfrm_thigh_r, loc=<33.5 0 0>, rot=<0.0006 -0.0161 -0.0391 -0.9991>}
 dfrm_foot_r {parent=dfrm_calf_r, loc=<35.8914 -0.2597 -0.294>, rot=<0.0432 -0.0278 0.0277 -0.9983>}
 dfrm_leg_spline_09_r_ctrl_null {parent=dfrm_foot_r, type=Null, loc=<0 0 0>, rot=<-0.0195 -0.0445 0.9983 -0.0322>}
 dfrm_leg_spline_09_r_ctrl {controlType=EULER_TRANSFORM, parent=dfrm_leg_spline_09_r_ctrl_null, type=Control, loc=<0 0 0>, rot=<0 0 0 1>}
 dfrm_thigh_l {parent=pelvis, loc=<-14.5412 -1.3407 -11.2>, rot=<0.0004 0.0397 0.0229 -0.9989>}
 dfrm_calf_l {parent=dfrm_thigh_l, loc=<-33.5 0 0>, rot=<0.0006 -0.0161 -0.0391 -0.9991>}
 dfrm_foot_l {parent=dfrm_calf_l, loc=<-35.8913 0.2597 0.294>, rot=<0.0432 -0.0278 0.0277 -0.9983>}
 dfrm_leg_spline_09_l_ctrl_null {parent=dfrm_foot_l, type=Null, loc=<0 0 0>, rot=<0.0195 0.0445 -0.9983 0.0322>}
 dfrm_leg_spline_09_l_ctrl {controlType=EULER_TRANSFORM, parent=dfrm_leg_spline_09_l_ctrl_null, type=Control, loc=<0 0 0>, rot=<0 0 0 1>}
 Tail {parent=pelvis, type=Null, loc=<-3.2227 -15.7579 1.4019>, rot=<-0.0296 0.7091 0.0916 0.6985>}
 dfrm_leg_spline_01_l_ctrl_null {parent=pelvis, type=Null, loc=<-14.5412 -1.3407 -11.2>, rot=<0.0448 -0.0002 -0.9985 -0.0318>}
 dfrm_leg_spline_01_l_ctrl {controlType=EULER_TRANSFORM, parent=dfrm_leg_spline_01_l_ctrl_null, type=Control, loc=<0 0 0>, rot=<0 0 0 1>}
 dfrm_leg_spline_03_l_ctrl_null {parent=pelvis, type=Null, loc=<-31.795 -0.3077 -12.7146>, rot=<-0.0382 0.0006 0.9991 0.0201>}
 dfrm_leg_spline_03_l_ctrl {controlType=EULER_TRANSFORM, parent=dfrm_leg_spline_03_l_ctrl_null, type=Control, loc=<0 0 0>, rot=<0 0 0 1>}
 dfrm_leg_spline_05_l_ctrl_null {parent=pelvis, type=Null, loc=<-49.0945 0.198 -13.9293>, rot=<-0.025 0.001 0.9997 -0.0029>}
 dfrm_leg_spline_05_l_ctrl {controlType=EULER_TRANSFORM, parent=dfrm_leg_spline_05_l_ctrl_null, type=Control, loc=<0 0 0>, rot=<0 0 0 1>}
 dfrm_leg_spline_07_l_ctrl_null {parent=pelvis, type=Null, loc=<-66.425 -0.0576 -14.709>, rot=<0.0167 -0.0009 -0.9997 0.0174>}
 dfrm_leg_spline_07_l_ctrl {controlType=EULER_TRANSFORM, parent=dfrm_leg_spline_07_l_ctrl_null, type=Control, loc=<0 0 0>, rot=<0 0 0 1>}
 dfrm_arm_spline_03_r_ctrl_null {parent=pelvis, type=Null, loc=<33.0441 -6.6985 26.9894>, rot=<-0.9118 -0.0226 0.409 -0.0276>}
 dfrm_arm_spline_03_r_ctrl {controlType=EULER_TRANSFORM, parent=dfrm_arm_spline_03_r_ctrl_null, type=Control, loc=<0 0 0>, rot=<0 0 0 1>}
 dfrm_arm_spline_05_r_ctrl_null {parent=pelvis, type=Null, loc=<24.8427 -6.3118 36.1483>, rot=<-0.9069 0.0972 0.4019 -0.081>}
 dfrm_arm_spline_05_r_ctrl {controlType=EULER_TRANSFORM, parent=dfrm_arm_spline_05_r_ctrl_null, type=Control, loc=<0 0 0>, rot=<0 0 0 1>}
 dfrm_arm_spline_07_r_ctrl_null {parent=pelvis, type=Null, loc=<16.9278 -2.3151 44.6121>, rot=<-0.8815 0.2343 0.3848 -0.1413>}
 dfrm_arm_spline_07_r_ctrl {controlType=EULER_TRANSFORM, parent=dfrm_arm_spline_07_r_ctrl_null, type=Control, loc=<0 0 0>, rot=<0 0 0 1>}
 dfrm_leg_spline_01_r_ctrl_null {parent=pelvis, type=Null, loc=<-14.5412 -1.3407 11.2>, rot=<0.0002 0.0448 -0.0318 0.9985>}
 dfrm_leg_spline_01_r_ctrl {controlType=EULER_TRANSFORM, parent=dfrm_leg_spline_01_r_ctrl_null, type=Control, loc=<0 0 0>, rot=<0 0 0 1>}
 dfrm_leg_spline_03_r_ctrl_null {parent=pelvis, type=Null, loc=<-31.795 -0.3077 12.7146>, rot=<-0.0006 -0.0382 0.0201 -0.9991>}
 dfrm_leg_spline_03_r_ctrl {controlType=EULER_TRANSFORM, parent=dfrm_leg_spline_03_r_ctrl_null, type=Control, loc=<0 0 0>, rot=<0 0 0 1>}
 dfrm_leg_spline_05_r_ctrl_null {parent=pelvis, type=Null, loc=<-49.0945 0.198 13.9293>, rot=<-0.001 -0.025 -0.0029 -0.9997>}
 dfrm_leg_spline_05_r_ctrl {controlType=EULER_TRANSFORM, parent=dfrm_leg_spline_05_r_ctrl_null, type=Control, loc=<0 0 0>, rot=<0 0 0 1>}
 dfrm_leg_spline_07_r_ctrl_null {parent=pelvis, type=Null, loc=<-66.425 -0.0576 14.709>, rot=<-0.0009 -0.0167 -0.0174 -0.9997>}
 dfrm_leg_spline_07_r_ctrl {controlType=EULER_TRANSFORM, parent=dfrm_leg_spline_07_r_ctrl_null, type=Control, loc=<0 0 0>, rot=<0 0 0 1>}
 thigh_l_null {parent=pelvis, type=Null, loc=<-2.5693 -0.0109 -7.7501>, rot=<0.0012 0.0572 0.0155 -0.9982>}
 thigh_r_null {parent=pelvis, type=Null, loc=<-2.5693 -0.0109 7.7501>, rot=<0.0572 -0.0012 0.9982 0.0155>}
 thigh_l_ctrl {controlType=EULER_TRANSFORM, type=Control, loc=<0 0 0>, rot=<0 0 0 1>}
 calf_l_ctrl {controlType=EULER_TRANSFORM, parent=thigh_l_ctrl, type=Control, loc=<0 0 0>, rot=<0 0 0 1>}
 thigh_r_ctrl {controlType=EULER_TRANSFORM, type=Control, loc=<0 0 0>, rot=<0 0 0 1>}
 calf_r_ctrl {controlType=EULER_TRANSFORM, parent=thigh_r_ctrl, type=Control, loc=<0 0 0>, rot=<0 0 0 1>}
 upperarm_r_ctrl {controlType=EULER_TRANSFORM, type=Control, loc=<0 0 0>, rot=<0 0 0 1>}
 lowerarm_r_ctrl {controlType=EULER_TRANSFORM, parent=upperarm_r_ctrl, type=Control, loc=<0 0 0>, rot=<0 0 0 1>}
 upperarm_l_ctrl {controlType=EULER_TRANSFORM, type=Control, loc=<0 0 0>, rot=<0 0 0 1>}
 lowerarm_l_ctrl {controlType=EULER_TRANSFORM, parent=upperarm_l_ctrl, type=Control, loc=<0 0 0>, rot=<0 0 0 1>}
]
variables [
 L_armHidden {type=bool, default=False}
 L_armSpline_ctrls {type="TArray<FRigElementKey>"}
 L_calfUp {type=double, default=0}
 L_legSpline_ctrls {type="TArray<FRigElementKey>"}
 L_thighUp {type=double, default=0}
 R_armHidden {type=bool, default=False}
 R_armSpline_ctrls {type="TArray<FRigElementKey>"}
 R_calfUp {type=double, default=0}
 R_legSpline_ctrls {type="TArray<FRigElementKey>"}
 R_thighUp {type=double, default=0}
 isCrouching_CR {type=bool, default=False}
]
