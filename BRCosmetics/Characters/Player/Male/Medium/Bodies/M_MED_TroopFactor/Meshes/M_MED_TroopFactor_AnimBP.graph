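# Post-processing style AnimBP for the TroopFactor male medium body that adapts the attached parent mesh's pose, mainly posing its tail.
# Copied pose feeds a Control Rig given arm-hidden and crouch flags, then a front-end backpack offset.
# Chained additive dfrm_tail bone edits blend in by movement state, front end, skydiving and parachute.
# Event Graph derives combined state booleans each update; extraTailPush is set but unused here.
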AnimBlueprint /BRCosmetics/Characters/Player/Male/Medium/Bodies/M_MED_TroopFactor/Meshes/M_MED_TroopFactor_AnimBP
target_skeleton=&/M_MED_TroopFactor_Skeleton
graphs [
 AnimGraph {nodes [
   n1 = Control Rig!(
    Source=Copy Pose From Mesh() {bUseAttachedParent=true}
    Alpha
    L_armHidden:bool=Get bIsLeftArmHidden() -> (bIsLeftArmHidden:bool)
    R_armHidden:bool=Get bIsRightArmHidden() -> (bIsRightArmHidden:bool)
    isCrouching_CR:bool=Get isCrouchingButNotWalking() -> (isCrouchingButNotWalking:bool)
   ) -> (Pose) {controlRigAssetReference {blueprintRigClass=&/M_MED_TroopFactor_CtrlRig.M_MED_TroopFactor_CtrlRig_C}}
   n2 = "Transform (Modify) Bone - Bone: attach_backpack"!(
    ComponentPose=Local To Component(LocalPose=@n1)
    bAlphaBoolEnabled:bool=Get bIsInFrontEnd()
    Translation="0,2,0"
    Rotation="0,0,0"
    Scale="1,1,1"
   ) -> (Pose) {
    alphaInputType=Bool
    rotationMode=BMM_Additive
    scaleMode=BMM_Additive
    translationMode=BMM_Additive
    boneToModify {boneName=attach_backpack}
    translation {y=2}
   }
   n3 = "Transform (Modify) Bone - Bone: dfrm_tail"!(
    ComponentPose=@n2
    bAlphaBoolEnabled:bool=Get isGoingSlightlyFast() -> (isGoingSlightlyFast:bool)
    Translation="0,0,0"
    Rotation="0,0,30"
    Scale="1,1,1"
   ) -> (Pose) {
    alphaInputType=Bool
    rotationMode=BMM_Additive
    scaleMode=BMM_Additive
    translationMode=BMM_Additive
    alphaBoolBlend {blendInTime=1, blendOutTime=2.5}
    boneToModify {boneName=dfrm_tail}
    rotation {roll=30}
   }
   n4 = "Transform (Modify) Bone - Bone: dfrm_tail"(ComponentPose=@n3, bAlphaBoolEnabled=Get bIsTacticalSprinting(), Translation="0,0,0", Rotation="0,0,70", Scale="1,1,1") {
    alphaInputType=Bool
    rotationMode=BMM_Additive
    scaleMode=BMM_Additive
    translationMode=BMM_Additive
    alphaBoolBlend {blendInTime=0.5, blendOutTime=2.5}
    boneToModify {boneName=dfrm_tail}
    rotation {roll=70}
   }
   n5 = "Transform (Modify) Bone - Bone: dfrm_tail"(
    ComponentPose=@n4
    bAlphaBoolEnabled=Get isCrouchSprintOrDBNO() -> (isCrouchSprintOrDBNO:bool)
    Translation="0,0,0"
    Rotation="0,0,-13"
    Scale="1,1,1"
   ) {
    alphaInputType=Bool
    rotationMode=BMM_Additive
    scaleMode=BMM_Additive
    translationMode=BMM_Additive
    alphaBoolBlend {blendInTime=1.5, blendOutTime=1}
    boneToModify {boneName=dfrm_tail}
    rotation {roll=-13}
   }
   n6 = "Transform (Modify) Bone - Bone: dfrm_tail"(ComponentPose=@n5, bAlphaBoolEnabled=Get bIsInFrontEnd(), Translation="3,0,0", Rotation="0,0,-15", Scale="0.9,1,1") {
    alphaInputType=Bool
    rotationMode=BMM_Additive
    scaleMode=BMM_Additive
    translationMode=BMM_Additive
    alphaBoolBlend {blendInTime=1.5, blendOutTime=1}
    boneToModify {boneName=dfrm_tail}
    rotation {roll=-15}
    scale {x=0.9}
    translation {x=3}
   }
   n7 = "Transform (Modify) Bone - Bone: dfrm_tail"(ComponentPose=@n6, bAlphaBoolEnabled=Get bIsSkyDiving(), Translation="0,0,8", Rotation="0,0,15", Scale="0.9,1,1") {
    alphaInputType=Bool
    rotationMode=BMM_Additive
    scaleMode=BMM_Additive
    translationMode=BMM_Additive
    alphaBoolBlend {blendInTime=0.3, blendOutTime=0.3}
    boneToModify {boneName=dfrm_tail}
    rotation {roll=15}
    scale {x=0.9}
    translation {z=8}
   }
   n8 = "Transform (Modify) Bone - Bone: dfrm_tail"(
    ComponentPose=@n7
    bAlphaBoolEnabled=Get bIsParachuteOpened()
    Translation="-0.000,-9.642,-6.344"
    Rotation="0.000,0.000,-27.617"
    Scale="0.9,1,1"
   ) {
    alphaInputType=Bool
    rotationMode=BMM_Additive
    scaleMode=BMM_Additive
    translationMode=BMM_Additive
    alphaBoolBlend {blendInTime=0.3, blendOutTime=0.3}
    boneToModify {boneName=dfrm_tail}
    rotation {roll=-27.617}
    scale {x=0.9}
    translation {y=-9.642, z=-6.344}
   }
   Output Pose(Result=Component To Local(ComponentPose=@n8)) {layerGroup=None}
  ]}
 EventGraph {nodes [
   Event BlueprintUpdateAnimation()
   n1 = Get bIsCrouchSprinting()
   > Set isCrouchingButNotWalking(isCrouchingButNotWalking:bool=AND Boolean(A=Get bIsCrouching(), B=NOT Boolean(A=@n1))) -> (Output_Get:bool)
   n2 = Get bIsMoving2D()
   n3 = Get bIsInVehicle() -> (bIsInVehicle:bool)
   > Set isGoingSlightlyFast(isGoingSlightlyFast:bool=OR Boolean!(A=Get bIsCrouching(), B=Get bIsSkyDiving(), C:bool=Get bIsOnGrindRail(), D:bool=AND Boolean(A=@n3, B=@n2)) -> (ReturnValue)) -> (Output_Get:bool)
   > Set isCrouchSprintOrDBNO(isCrouchSprintOrDBNO:bool=OR Boolean!(A=Get bIsCrouchSprinting(), B=Get bIsDBNO()) -> (ReturnValue)) -> (Output_Get:bool)
   n4 = Get bIsCrouching()
   > Set extraTailPush(extraTailPush:bool=OR Boolean(A=Get bIsJumping(), B=AND Boolean(A=@n4, B=false))) -> (Output_Get:bool)
  ]}
]
variables [
 extraTailPush {type=boolean}
 Is Crouching {type=boolean}
 isCrouchingButNotWalking {type=boolean}
 isCrouchSprintOrDBNO {type=boolean}
 isGoingSlightlyFast {type=boolean}
]
